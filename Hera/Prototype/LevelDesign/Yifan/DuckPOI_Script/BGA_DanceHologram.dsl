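; Prototype building actor for a dance hologram that switches on when its button-press event fires.
; On authority the event sets ButtonIsPressed, whose RepNotify makes the point light and hologram visible; nothing here switches them off.
; On remote the same event spawns a borrowed mushroom-bounce Niagara effect at the arrow component.
; The actor disables damage at BeginPlay and dies silently when its linked floor is destroyed.
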(event EventBeginPlay
  (Variables|BuildingActor|SetbIsInvulnerable false)
  (Variables|Actor|SetbCanBeDamaged false))

(event Custom|ButtonPressEvent
  (Networking|SwitchHasAuthority
    (:Authority
      (Variables|Default|SetButtonIsPressed true))
    (:Remote
      ; fx
      (Niagara|SpawnSystematLocation "/Game/Athena/Artemis/Environments/BuildingActors/Daisy/FX/NS_Daisy_Mushroom_Bounce.NS_Daisy_Mushroom_Bounce" (Transformation|GetWorldLocation (Variables|Default|GetArrow))))))

(event OnLinkedActorDestroyed(FortLinkToActor)
  (params (DamageCauser :type Actor%20Object%20Reference))
  ; If the floor we're attached to is destroyed, destroy this
  (Gameplay|SilentDie false))

(fn OnRep_ButtonIsPressed
  (if (Variables|Default|GetButtonIsPressed)
    (Rendering|SetVisibility (Variables|Default|GetPointLight) true false)
    (Rendering|SetVisibility (Variables|Default|GetHologram) true false)))

(fn ConstructionScript)
