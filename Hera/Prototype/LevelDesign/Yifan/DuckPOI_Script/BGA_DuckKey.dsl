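; Prototype interactable duck key building actor for a duck POI, which gives out a key pickup when used.
; On BeginPlay the authority makes it invulnerable, and interaction prompt and availability come from variables.
; Interacting disables the button and calls ButtonPressEvent, where authority tosses a key pickup above the mesh while remote plays reveal sound and Niagara effect.
; The actor then silently dies after a brief delay, with debug strings printed along the way.
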(event EventBeginPlay
  ; Set the key to be indestrutible at the beginning
  (Networking|SwitchHasAuthority
    (:Authority
      (Variables|Actor|SetbCanBeDamaged false)
      (Variables|BuildingActor|SetbIsInvulnerable true))))

(event Interaction|EventBlueprintOnInteract
  (params (InteractingPawn :type Fort%20Pawn%20Object%20Reference) (InteractionBeingAttempted :type EInteractionBeingAttempted%20Enum :default "FirstInteraction"))
  ;[0.1324 0.05655 0.15 0.5] Add interaction button
  (Variables|Default|SetButtonCanBePressed false)
  (Development|PrintString "You get the duck key" true true "(R=0.000000,G=0.660000,B=1.000000,A=1.000000)" 3e+01)
  (CallFunction|ButtonPressEvent)
  (Utilities|FlowControl|Delay 0.05)
  (Gameplay|SilentDie false))

(event Custom|ButtonPressEvent
  ;[0.03371 0.1371 0.15 0.5] Spawn special effects and play music
  (Networking|SwitchHasAuthority
    (:Authority
      (bind _returnvalue (Inventory|SpawnPickupInWorldWithLevel (Variables|Default|GetKey) 0 1 (+ (Transformation|GetWorldLocation (Variables|Default|GetStaticMesh)) "0.000000,0.000000,50.000000") "0.000000,0.000000,1.000000" 0 true true false -1 "Tossed" "Unset" 0 false))
      (Development|PrintString (Utilities|String|Append "Key has spawned = " (Utilities|GetDisplayName _returnvalue)) true true "(R=0.000000,G=0.660000,B=1.000000,A=1.000000)" 5.0))
    (:Remote
      (Audio|PlaySoundatLocation "/Hera/Audio/Props/Duck_Key/SC_Duck_Key_Reveal_Cue.SC_Duck_Key_Reveal_Cue" (Transformation|GetActorLocation))
      (Niagara|SpawnSystematLocation "/Game/Athena/Artemis/Environments/BuildingActors/Daisy/FX/NS_Daisy_Mushroom_Bounce.NS_Daisy_Mushroom_Bounce" (Transformation|GetWorldLocation (Variables|Default|GetSpawnHere))))))

(fn BlueprintCanInteract
  (params (InteractingPawn :type Fort%20Pawn%20Object%20Reference) (InteractionBeingAttempted :type EInteractionBeingAttempted%20Enum :default "FirstInteraction") (InteractionType :type TInteractionType%20Enum :default "IT_NoInteraction"))
  (return (Variables|Default|GetButtonCanBePressed)))

(fn BlueprintGetInteractionString
  (params (InteractingPawn :type Fort%20Pawn%20Object%20Reference) (InteractionBeingAttempted :type EInteractionBeingAttempted%20Enum :default "FirstInteraction"))
  (return (Variables|Default|GetButtonText)))

(fn ConstructionScript)
