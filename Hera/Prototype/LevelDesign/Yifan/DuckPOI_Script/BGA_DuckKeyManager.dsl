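; Duck key manager building actor for the Duck POI that swaps one duck prop for a duck key when play begins.
; On the server only, it makes itself invulnerable, then continues if the hotfixable DuckVaultEnabled flag is on.
; It shuffles its duck list and takes the first entry, or uses a fixed index in debug mode.
; It records that duck's transform, destroys it, and spawns BGA_DuckKey there.
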(event EventBeginPlay
  (bind _listofducks (Variables|Default|GetListOfDucks))
  ;[0.1158 0.03435 0.15 0.5] Select a random duck prop on the map, destroy the prop and replace it with the duck key
  (Networking|SwitchHasAuthority
    (:Authority
      (Variables|Actor|SetbCanBeDamaged false)
      (Variables|BuildingActor|SetbIsInvulnerable true)
      (if (ScalableFloat|GetValueasBool (Variables|Hotfixable|GetDuckVaultEnabled))
        (if (Variables|Default|GetEnableDebugDuckKeySpawn)
          (Variables|Default|SetDuckToSpawnRef (Utilities|Array|Get(acopy) (Variables|Default|GetListOfDucks) (Variables|Default|GetDebugDuckKeySpawnIndex)))
          (else
            (Utilities|Array|Shuffle _listofducks)
            (Variables|Default|SetDuckToSpawnRef (Utilities|Array|Get(acopy) _listofducks))))
        (Variables|Default|SetDuckToReplaceWorldTransform (Transformation|GetActorTransform (Variables|Default|GetDuckToSpawnRef)))
        (Actor|DestroyActor (Variables|Default|GetDuckToSpawnRef))
        (Game|SpawnActorfromClass "/Hera/Prototype/LevelDesign/Yifan/DuckPOI_Script/BGA_DuckKey.BGA_DuckKey_C" (Variables|Default|GetDuckToReplaceWorldTransform) "AlwaysSpawn" "MultiplyWithRoot" 0 0 0)))))

(fn ConstructionScript)
