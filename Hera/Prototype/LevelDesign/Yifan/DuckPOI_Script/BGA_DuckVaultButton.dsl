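; Invulnerable Fortnite building actor for the duck POI: a vault button that opens its linked door for players holding a key item.
; The interaction check requires an unused button and a key in the player's inventory.
; Pressing disables the button, removes the key and opens the door after one second.
; The press event sets replicated state on authority, driving light/duck visibility via OnRep, and plays audio and Niagara FX on remotes.
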(event EventBeginPlay
  ; Set it to be invulnerable
  (Fortnite|SetCanbeDamaged self)
  (Variables|BuildingActor|SetbIsInvulnerable true))

(event Interaction|EventBlueprintOnInteract
  (params (InteractingPawn :type Fort%20Pawn%20Object%20Reference) (InteractionBeingAttempted :type EInteractionBeingAttempted%20Enum :default "FirstInteraction"))
  ;[0.15 0.1044 0.006406 0.5] On interact, trigger button press event
  (Variables|Default|SetButtonCanbePressed false)
  (CallFunction|ButtonPressEvent)
  (CallFunction|RemoveKey InteractingPawn)
  (Utilities|FlowControl|Delay 1.0)
  (Class|BGADuckVaultDoor|OpenDoor (Variables|Default|GetVaultDoorReference)))

(event Custom|ButtonPressEvent
  ;[0.04435 0.1221 0.15 0.5] After button press, do these
  (Networking|SwitchHasAuthority
    (:Authority
      (Variables|Default|SetButtonIsPressed true))
    (:Remote
      (Audio|Components|Audio|Play (Variables|Default|GetAudio) 0.0)
      (Niagara|SpawnSystematLocation "/Game/Athena/Artemis/Environments/BuildingActors/Daisy/FX/NS_Daisy_Mushroom_Bounce.NS_Daisy_Mushroom_Bounce" (Transformation|GetWorldLocation (Variables|Default|GetVFXlocation))))))

(fn BlueprintCanInteract
  (params (InteractingPawn :type Fort%20Pawn%20Object%20Reference) (InteractionBeingAttempted :type EInteractionBeingAttempted%20Enum :default "FirstInteraction") (InteractionType :type TInteractionType%20Enum :default "IT_NoInteraction"))
  ; Check if this player has a key
  (bind _asfort_player_pawn_athena (Utilities|Casting|CastToFortPlayerPawnAthena InteractingPawn)) ; CheckIfThisIsAPlayer
  (bind _asfort_player_controller_athena (Utilities|Casting|CastToFortPlayerControllerAthena (Pawn|GetController _asfort_player_pawn_athena))) ; Find this specific player controller's inventory
  (return (and (Variables|Default|GetButtonCanbePressed) (>= (Inventory|GetItemQuantityonInventoryOwner (Fortnite|GetFortPlayerControllerfromActor _asfort_player_controller_athena) (Variables|Default|GetKeyItem)) 1))))

(fn BlueprintGetInteractionString
  (params (InteractingPawn :type Fort%20Pawn%20Object%20Reference) (InteractionBeingAttempted :type EInteractionBeingAttempted%20Enum :default "FirstInteraction"))
  (return (Variables|Default|GetInteractionName)))

(fn GetBoundsComponents
  (params (SelfActor :type Actor%20Object%20Reference))
  (return (Utilities|Array|MakeArray (Variables|Default|GetTable))))

(fn OnRep_ButtonIsPressed
  (if (Variables|Default|GetButtonIsPressed)
    (Rendering|SetVisibility (Variables|Default|GetPointLight) true false)
    (Rendering|SetVisibility (Variables|Default|GetRubberDuck) true false)))

(fn RemoveKey
  (params (PlayerPawn :type Fort%20Pawn%20Object%20Reference))
  (bind _asfort_player_controller_athena (Utilities|Casting|CastToFortPlayerControllerAthena (Pawn|GetController PlayerPawn)))
  (Inventory|RemoveItemfromInventoryOwner _asfort_player_controller_athena (Variables|Default|GetKeyItem)))

(fn ConstructionScript)
