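; Vault door building actor, under a Duck POI prototype, that slides downward when its OpenDoor event fires.
; On BeginPlay it makes itself invulnerable and caches the door mesh's relative start location and a lowered end location.
; OpenDoor runs once, playing a timeline that lerps the mesh between those locations.
; The timeline's DoorSound event track triggers the audio component as the door moves.
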(event EventBeginPlay
  (bind _doormesh (Variables|Default|GetDoorMesh))
  (bind _relativelocation (Variables|/Script/Engine.SceneComponent|GetRelativeLocation _doormesh))
  ;[0.15 0.1006 0.004531 0.5] Set the door to be indestructible. Set it to move downward once activated
  (Variables|BuildingActor|SetbIsInvulnerable true)
  (Fortnite|SetCanbeDamaged self)
  (Variables|Default|SetDoorStartLoc _relativelocation)
  (Variables|Default|SetDoorEndLoc (+ _relativelocation "0.000000,0.000000,-365.000000")))

(event Custom|OpenDoor
  ;[0.5198 0.1978 0.6865 0.5] Timelines for door opening and door sound
  (Utilities|FlowControl|DoOnce false
    (:Completed
      (bind (_direction _doormovingtrack) (Timeline|TimelineForDoor Play)
        (:Update
          (Transformation|SetRelativeLocation (Variables|Default|GetDoorMesh) (Math|Vector|Lerp(Vector) (Variables|Default|GetDoorStartLoc) (Variables|Default|GetDoorEndLoc) _doormovingtrack) false false))
        (:DoorSound
          (Networking|SwitchHasAuthority
            (:Authority
              _)
            (:Remote
              _))
          (Audio|Components|Audio|Play (Variables|Default|GetAudio) 0.0))) ; Used the "door soud" event track to play sound as the door opens
    )))

(fn ConstructionScript)
