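; Building game actor for the Duck POI party: an invulnerable trigger volume that, by its name, supports emote-based shield regen, though no regen logic is here.
; Its box stays collisionless until PartyHasStarted sets the party flag and enables query collision on the server.
; Server-side overlaps maintain a player pawn list and broadcast OnOverlappingPlayersUpdated.
; Helper functions report whether anyone overlaps or is emoting; the party flag's OnRep does nothing.
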(event EventBeginPlay
  ;[0.15 0.006874 0.000156 0.5] Set the bp to Invulnerable
  (Variables|BuildingActor|SetbIsInvulnerable true)
  (Variables|Actor|SetbCanBeDamaged false)
  (Collision|SetCollisionEnabled (Variables|Default|GetBox) "NoCollision"))

(event Custom|PartyHasStarted
  ;[0.9281 0.8822 0 0.5] 1. Custom event to set party started boolean value to true. 2. Swap dance floor material
  (Variables|Default|SetPartyHasStrated true)
  (Networking|SwitchHasAuthority
    (:Authority
      (Collision|SetCollisionEnabled (Variables|Default|GetBox) "QueryOnly"))))

(event OnComponentBeginOverlap(Box)
  (params (OverlappedComponent :type Primitive%20Component%20Object%20Reference) (OtherActor :type Actor%20Object%20Reference) (OtherComp :type Primitive%20Component%20Object%20Reference) (OtherBodyIndex :type Integer) (bFromSweep :type Boolean) (SweepResult :type Hit%20Result%20Structure%20%28by%20ref%29))
  ;[0.3909 0.3112 0.999 0.5] Check keep track of players that are overlapping
  (Networking|SwitchHasAuthority
    (:Authority
      (bind _asfort_player_pawn_athena (Utilities|Casting|CastToFortPlayerPawnAthena OtherActor))
      (Utilities|Array|AddUnique (Variables|Default|GetOverlappingPlayers) _asfort_player_pawn_athena)
      (Default|CallOnOverlappingPlayersUpdated self (Variables|Default|GetOverlappingPlayers)))))

(event OnComponentEndOverlap(Box)
  (params (OverlappedComponent :type Primitive%20Component%20Object%20Reference) (OtherActor :type Actor%20Object%20Reference) (OtherComp :type Primitive%20Component%20Object%20Reference) (OtherBodyIndex :type Integer))
  ;[0.3909 0.3112 0.999 0.5] Check keep track of players that are overlapping
  (Networking|SwitchHasAuthority
    (:Authority
      (bind _asfort_player_pawn_athena (Utilities|Casting|CastToFortPlayerPawnAthena OtherActor))
      (Utilities|Array|RemoveItem (Variables|Default|GetOverlappingPlayers) _asfort_player_pawn_athena)
      (Default|CallOnOverlappingPlayersUpdated self (Variables|Default|GetOverlappingPlayers)))))

(fn isAnyoneEmoting
  (params (PlayerActors :type Array%20of%20Actor%20Object%20References))
  (bind _array_element (Utilities|Array|ForEachLoopwithBreak PlayerActors)
    (:LoopBody
      (bind _asfort_player_pawn_athena (Utilities|Casting|CastToFortPlayerPawnAthena _array_element))
      (bind _asfort_player_controller_athena (Utilities|Casting|CastToFortPlayerControllerAthena (Pawn|GetController _asfort_player_pawn_athena)))
      (if (Emote|IsPlayingEmote _asfort_player_controller_athena)
        (LocalVariables|SetIsEmotingLocal true)))
    (:Completed
      _))
  (return (LocalVariables|GetIsEmotingLocal)))

(fn IsAnyPlayerOverlapping
  (return (Utilities|Array|IsNotEmpty (Variables|Default|GetOverlappingPlayers))))

(fn OnRep_PartyHasStrated
  (if (Variables|Default|GetPartyHasStrated)
    _))

(fn ConstructionScript)
