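; Invulnerable, one-press party switch for the duck POI that starts a dance party healing players who emote on the dance floor.
; Pressing notifies linked emote triggers, while RepNotify drives holograms, music, fog and lit floor materials.
; Trigger overlaps toggle a server timer that applies the heal effect once per emoting, safe-zone player each tick.
; Losing every floor piece ends the party; losing the switch's support removes it, or hides it if pressed.
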(event EventBeginPlay
  ;[0.15 0.005781 0.005781 0.5] Set this switch to invulnerable
  (Variables|Actor|SetbCanBeDamaged false)
  (Variables|BuildingActor|SetbIsInvulnerable true)
  (if (ScalableFloat|GetValueasBool (Variables|Default|GetEnabledHF))
    (CallFunction|BindTriggerEvents)
    (else
      (Variables|Interaction|SetbAllowInteract false))))

(event Interaction|EventBlueprintOnInteract
  (params (InteractingPawn :type Fort%20Pawn%20Object%20Reference) (InteractionBeingAttempted :type EInteractionBeingAttempted%20Enum :default "FirstInteraction"))
  (Variables|Default|SetButtonCanbePressed false)
  ;[0.02391 0.15 0.06087 0.5] Play music, turn on holograms, activate dance fog
  (Variables|Default|SetButtonIsPressed true)
  ;[0.2244 0.7448 1 0.5] Ask each emote detection trigger to toggle on partystarted boolean
  (for _array_element (Variables|Default|GetEmoteTriggerReference)
    (Utilities|IsValid _array_element
      (:Is%20Valid
        (Class|BGAEmoteShieldRegenTrigger|PartyHasStarted _array_element))))
  ; Start healing players
  (if (CallFunction|IsAnyPlayerOverlappingTriggers)
    (CallFunction|StartHealUpdateLoop)))

(event Custom|StopHealUpdateLoop
  ; Try to heal overlapping players every X seconds
  (Networking|SwitchHasAuthority
    (:Authority
      (Utilities|Time|ClearandInvalidateTimerbyHandle (Variables|Default|GetHealUpdateTimer)))))

(event Custom|StartHealUpdateLoop
  ; Try to heal overlapping players every X seconds
  (Networking|SwitchHasAuthority
    (:Authority
      (if (not (Variables|Default|GetDanceFloorRemoved))
        (bind _returnvalue (Utilities|Time|SetTimerbyEvent (EventDispatchers|CreateEvent :!event TryHealOverlappingPlayers) (Math|Float|Clamp(Float) (ScalableFloat|GetValueatLevel (Variables|Default|GetUpdateTickRateHF)) 0.033 (ScalableFloat|GetValueatLevel (Variables|Default|GetUpdateTickRateHF))) true))
        (Variables|Default|SetHealUpdateTimer _returnvalue)))))

(event OnLinkedActorDestroyed(FortLinkToActor)
  (params (DamageCauser :type Actor%20Object%20Reference))
  ; If the floor this actor is on, remove it if the party hasn't started, otherwise hide it
  (if (Variables|Default|GetButtonIsPressed)
    (Variables|Default|SetLinkedActorDestroyed true)
    (else
      (Gameplay|SilentDie false))))

(fn BindTriggerEvents
  (Networking|SwitchHasAuthority
    (:Authority
      (for _array_element (Variables|Default|GetEmoteTriggerReference)
        (Utilities|IsValid _array_element
          (:Is%20Valid
            (Default|BindEventtoOnOverlappingPlayersUpdated _array_element (EventDispatchers|CreateEvent :!event OnTriggerOverlappingPlayersUpdated)))))
      (for _array_element_2 (Variables|Default|GetDanceFloorActorReferences)
        (Game|BindEventtoOnDestroyed _array_element_2 (EventDispatchers|CreateEvent :!event OnDanceFloorPieceDestroyed))
        (bind _asbuilding_actor (Utilities|Casting|CastToBuildingActor _array_element_2))
        (EventDispatchers|BindEventtoOnDamaged _asbuilding_actor (EventDispatchers|CreateEvent :!event OnDanceFloorPieceDamaged))))))

(fn BlueprintCanInteract
  (params (InteractingPawn :type Fort%20Pawn%20Object%20Reference) (InteractionBeingAttempted :type EInteractionBeingAttempted%20Enum :default "FirstInteraction") (InteractionType :type TInteractionType%20Enum :default "IT_NoInteraction"))
  (if (or (Variables|Default|GetDanceFloorRemoved) (Variables|Default|GetLinkedActorDestroyed))
    (return false)
    (else
      (return (Variables|Default|GetButtonCanbePressed)))))

(fn BlueprintGetInteractionString
  (params (InteractingPawn :type Fort%20Pawn%20Object%20Reference) (InteractionBeingAttempted :type EInteractionBeingAttempted%20Enum :default "FirstInteraction"))
  (return (Variables|Default|GetInteractionName)))

(fn HandleDanceFloorRemoved
  ; If all the pieces of the dance floor are removed, stop healing and FX
  (if (not (Variables|Default|GetDanceFloorRemoved))
    (Variables|Default|SetDanceFloorRemoved true)
    (CallFunction|StopHealUpdateLoop)
    (Variables|Default|SetButtonCanbePressed false)))

(fn IsAnyPlayerOverlappingTriggers
  (for _array_element (Variables|Default|GetEmoteTriggerReference)
    (if (Class|BGAEmoteShieldRegenTrigger|IsAnyPlayerOverlapping _array_element)
      (return true)))
  (return false))

(fn OnDanceFloorPieceDamaged
  (params (DamagedActor :type Actor%20Object%20Reference) (Damage :type Float%20%28single-precision%29) (InstigatedBy :type Controller%20Object%20Reference) (DamageCauser :type Actor%20Object%20Reference) (HitLocation :type Vector) (FHitComponent :type Primitive%20Component%20Object%20Reference) (BoneName :type Name) (Momentum :type Vector))
  (bind _asbuilding_actor (Utilities|Casting|CastToBuildingActor DamagedActor))
  (if (<= (Health|GetHealth _asbuilding_actor) 0)
    (CallFunction|OnDanceFloorPieceDestroyed (Variables|GetDamagedActor))))

(fn OnDanceFloorPieceDestroyed
  (params (DestroyedActor :type Actor%20Object%20Reference))
  ; Check if all the pieces of the dance floor are removed.  If they are, stop the party.
  (Utilities|Array|RemoveItem (Variables|Default|GetDanceFloorActorReferences) (Variables|GetDestroyedActor))
  (if (Utilities|Array|IsEmpty (Variables|Default|GetDanceFloorActorReferences))
    (CallFunction|HandleDanceFloorRemoved)))

(fn OnRep_ButtonIsPressed
  (if (not (Variables|Default|GetDanceFloorRemoved))
    (if (Variables|Default|GetButtonIsPressed)
      (Utilities|FlowControl|Sequence
        (:then_0
          ; Turn on holograms
          (for _array_element (Variables|Default|GetHologramReference)
            (Class|BGADanceHologram|ButtonPressEvent _array_element)))
        (:then_1
          ; Music
          (Audio|AddEventbyComponent (GameInstanceSubsystems|GetMusicEventSubsystem) (Variables|Default|GetMusic) "(TagName=\"MusicEventSubsystem.Event.POI.Special\")" "(GameplayTags=)" 0 "")
          (Audio|Components|Audio|Play (Variables|Default|GetButtonPressedAudio) 0.0))
        (:then_2
          ; FX Fog
          (bind _returnvalue (Niagara|SpawnSystematLocation "/Game/Environments/Hera/FX/DuckYacht/NS_Env_DuckYacht_Dance_Fog.NS_Env_DuckYacht_Dance_Fog" (Transformation|GetWorldLocation (Variables|Default|GetFX_SpawnLoc))))
          (Variables|Default|SetFogVFX _returnvalue))
        (:then_3
          ; Change dance floor
          (for _array_element_2 (Variables|Default|GetDanceFloorActorReferences)
            (Utilities|IsValid _array_element_2
              (:Is%20Valid
                (for _array_element_3 (Actor|GetComponentsByClass _array_element_2 "/Script/Engine.StaticMeshComponent")
                  (Rendering|Material|SetMaterial _array_element_3 1 "/Hera/Sets/WaddleJam/TechArt/MI_Base_WaddleDanceFloor_ON.MI_Base_WaddleDanceFloor_ON")
                  (Rendering|Material|SetVectorParameterValueonMaterials _array_element_3 "Dance Floor Position" (Variables|Default|GetDance%20Floor%20Origin)))))))))))

(fn OnRep_DanceFloorRemoved
  (if (Variables|Default|GetDanceFloorRemoved)
    (Variables|Interaction|SetbAllowInteract false)
    (Audio|Components|Audio|FadeOut (Variables|Default|GetMusic) 0.5 0.0 "Linear" :!class /Script/Engine.AudioComponent)
    (Components|DestroyComponent (Variables|Default|GetFogVFX))))

(fn OnRep_LinkedActorDestroyed
  (bind _switchmesh (Variables|Default|GetSwitchMesh))
  (if (Variables|Default|GetLinkedActorDestroyed)
    (Development|SetHiddeninGame _switchmesh true false)
    (Collision|SetCollisionEnabled _switchmesh "NoCollision")))

(fn OnTriggerOverlappingPlayersUpdated
  (params (TriggerActor :type BGA%20Emote%20Shield%20Regen%20Trigger%20Object%20Reference) (Players :type Array%20of%20Fort%20Player%20Pawn%20Object%20References))
  (if (CallFunction|IsAnyPlayerOverlappingTriggers)
    (if (not (Utilities|Time|IsValidTimerHandle (Variables|Default|GetHealUpdateTimer)))
      (CallFunction|StartHealUpdateLoop))
    (else
      (CallFunction|StopHealUpdateLoop))))

(fn TryHealOverlappingPlayers
  (Networking|SwitchHasAuthority
    (:Authority
      (if (not (Variables|Default|GetDanceFloorRemoved))
        (for _array_element (Variables|Default|GetEmoteTriggerReference)
          (for _array_element_2 (Variables|/Hera/Prototype/LevelDesign/Yifan/DuckPOI_Script/BGA_EmoteShieldRegenTrigger.BGA_EmoteShieldRegenTrigger_C|GetOverlappingPlayers _array_element)
            (LocalVariables|SetTempPlayer _array_element_2)
            (if (not (Utilities|Array|ContainsItem (LocalVariables|GetPlayersChecked) (LocalVariables|GetTempPlayer)))
              (Utilities|Array|Add (LocalVariables|GetPlayersChecked) (LocalVariables|GetTempPlayer))
              (bind _output_get (LocalVariables|SetTempASC (Ability|GetFortAbilitySystemComponent (LocalVariables|GetTempPlayer))))
              (Utilities|IsValid _output_get
                (:Is%20Valid
                  (if (GameplayTags|HasMatchingGameplayTag (LocalVariables|GetTempASC) "(TagName=\"Gameplay.InsideSafeZone\")")
                    (bind _asfort_player_controller_athena (Utilities|Casting|CastToFortPlayerControllerAthena (Pawn|GetController (LocalVariables|GetTempPlayer))))
                    (if (Emote|IsPlayingEmote _asfort_player_controller_athena)
                      (GameplayEffects|FortApplyGameplayEffectToTarget :self (Ability|GetFortAbilitySystemComponent self) :GameplayEffectClass (Variables|Default|GetHealEffect) :TargetOfEffect (LocalVariables|GetTempASC) :OptionalEffectCauser self :Level 0.0)
                      (CustomVerb|BroadcastCustomVerbMessage (CustomVerb|MakeCustomVerbMessage (LocalVariables|GetTempPlayer) "(TagName=\"CustomVerb.DuckDance\")")))))))))))))

(fn ConstructionScript
  (Variables|Default|SetDance%20Floor%20Origin (Transformation|GetWorldLocation (Variables|Default|GetArrow))))
