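; B_Item_DuckKey is a prototype Blueprint, apparently a key item for the Duck POI, that currently only logs its own lifecycle.
; BeginPlay calls the parent implementation, then prints an "added" debug message, and Destroyed prints "removed".
; Tick is a disabled stub and the construction script only calls its parent, so no key or unlock logic exists in these graphs.
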(event EventBeginPlay
  (|CallParentFunction)
  (Development|PrintString "added" true true "(R=0.000000,G=0.660000,B=1.000000,A=1.000000)" 2e+01))

(event EventTick
  (params (DeltaSeconds :type Float%20%28single-precision%29))
  (|CallParentFunction DeltaSeconds) ; This node is disabled and will not be called.\nDrag off pins to build functionality.
)

(event EventDestroyed
  (Development|PrintString "removed" true true "(R=0.000000,G=0.660000,B=1.000000,A=1.000000)" 2e+01))

(fn ConstructionScript
  (|CallParentFunction))
